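# Niagara system for the Lion's Mane vehicle boost, trailing jellyfish and bubbles from skeletal-mesh sockets, partitioned by port count.
# Bubbles merges three spawn groups through switch helpers and visibility tags; Jellyfish mixes mesh and sprite renderers with pulsing scale curves.
# A frontend-preview flag switches velocity, acceleration and lifetime to fake backward motion.
# Explosion's spawn rate and all painted-variant overrides are disabled; Bubbles emits death events without visible handlers.
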NiagaraSystem /VehicleCosmetics/Boosts/Boost_LionsMane01/FX/NS_VFX_LionsMane01
dependencies {
 usedDataInterfaces [
  &NiagaraDataInterfaceCurve
  &NiagaraDataInterfaceVector2DCurve
  &NiagaraDataInterfaceSpriteRendererInfo
  &NiagaraDataInterfaceVectorCurve
  &NiagaraDataInterfaceSkeletalMesh
 ]
 usedDynamicInputs []
 usedModules [
  &self:SystemSpawnScript.SystemScriptSource.SystemScriptGraph.NiagaraNodeAssignment_0.SetVariables_#1
  &SystemState
 ]
}
emitters {
 Bubbles {
  emitterUpdateScript [
   &EmitterState()
   &VC_FilteredSocketCount()
   &SpawnRate(SpawnRate=80, Use Spawn Probability=false)
   &SpawnRate(SpawnGroup=1, SpawnRate=60, Use Spawn Probability=false) {module=SpawnRate001}
   &SpawnRate(SpawnGroup=2, SpawnRate=50, Use Spawn Probability=false) {module=SpawnRate002}
  ]
  particleSpawnScript [
   /SetVariables(Particles.VisibilityTag=$Emitter.SpawnGroup) {isSetParameters=true}
   &VC_PaintedVariantsOverride() {enabled=false}
   &InitializeParticle(
    Color Mode=::NewEnumerator0
    Lifetime=&MakeCustomFloatFromBool(
     Bool=$User.bIsFrontendPreview
     True Float=&MergeHelper_FloatSwitch(
      0=&RandomRangeFloat(Minimum=0.5, Maximum=1)
      1=&RandomRangeFloat(Minimum=0.4, Maximum=0.8)
      2=&RandomRangeFloat(Minimum=0.2, Maximum=0.5)
      3=?
      Index=&SpawnGroup()
      Num=::NewEnumerator1
     )
     False Float=0.3
    )
    Sprite Size Mode=::NewEnumerator3
    Uniform Sprite Size=&MergeHelper_FloatSwitch(
     0=&RandomRangeFloat(Minimum=20, Maximum=40)
     1=&RandomRangeFloat(Minimum=5, Maximum=15)
     2=&RandomRangeFloat(Minimum=10, Maximum=20)
     3=?
     Index=&SpawnGroup()
     Num=::NewEnumerator1
    )
    UsePositionOffset=false
    Write Material Random=true
   )
   &PartitionParticles(Number Of Partitions=$Emitter.PortCount)
   &SkeletalMeshLocation(
    Bone / Socket Index=$Particles.PartitionParticles.Partition
    Bone / Socket Sampling Mode=::NewEnumerator5
    MaskSpawnGroup=false
    Mesh Sampling Type=::NewEnumerator6
    Sampled Position Offset=<-35 0 0>
    Skeletal Mesh=$System.SkeletalMesh
   )
   &AddVelocity(Velocity=&MakeCustomVectorFromBool(
     Bool=$User.bIsFrontendPreview
     True Vector=&MergeHelper_VectorSwitch(
      0=<-500 0 0>
      1=<-1300 0 0>
      2=&RandomRangeVector(Minimum=<0 200 0>, Maximum=<-300 -200 300>)
      3=?
      Index=&SpawnGroup()
      Num=::NewEnumerator1
     )
     False Vector=&MergeHelper_VectorSwitch(
      0=<0 0 0>
      1=<0 0 0>
      2=&RandomRangeVector(Minimum=<0 200 0>, Maximum=<0 -200 300>)
      3=?
      Index=&SpawnGroup()
      Num=::NewEnumerator1
     )
    ))
   &ShapeLocation(MaskSpawnGroup=false, Override Seed=false, Sphere Radius=25, UseOwnerScale=true)
   &Color() {enabled=false}
  ]
  particleUpdateScript [
   &ParticleState()
   &ScaleSpriteSize(
    Scale Factor=&MergeHelper_Vector2Switch(
     0=&Vector2DFromCurve(
      Vector2Curve=DataInterface() {propertyValues {XCurve {keys [<at 0 0>,<at 0.2 1>,<at 1 1>]}, YCurve {keys [<at 0 0>,<at 0.2 2>,<at 0.5 2>,<at 1 1>]}}}
      ScaleCurve=<1 1>
     )
     1=&Vector2DFromCurve(
      Vector2Curve=DataInterface() {propertyValues {XCurve {keys [<at 0.4 2>,<at 0.8 1>,<at 1 2.1>]}, YCurve {keys [<at 0.4 1>,<at 0.8 1>,<at 1 2.1>]}}}
      ScaleCurve=<1 1>
     )
     2=&Vector2DFromCurve(Vector2Curve=DataInterface() {propertyValues {XCurve {keys [<at 0.8 1>,<at 1 2>]}, YCurve {keys [<at 0.8 1>,<at 1 2>]}}}, ScaleCurve=<1 1>)
     3=?
     4=?
     Index=&SpawnGroup()
     Num=::NewEnumerator1
    )
    Scale Sprite Size Mode=::NewEnumerator2
   )
   &CurlNoiseForce(Bias Noise=false, Noise Frequency=30, Noise Strength=200, Pan Noise=false)
   &Drag(
    Drag=&MergeHelper_FloatSwitch(0=2, 1=3, 2=5, 3=?, Index=&SpawnGroup(), Num=::NewEnumerator1)
    Use Linear Drag=true
    Use Rotational Drag=true
   )
   &AccelerationForce(
    Acceleration=&MakeCustomVectorFromBool(
     Bool=$User.bIsFrontendPreview
     True Vector=&MergeHelper_VectorSwitch(0=<-1000 0 0>, 1=<0 0 200>, 2=<0 0 0>, 3=?, Index=&SpawnGroup(), Num=::NewEnumerator1)
     False Vector=<0 0 0>
    )
    Coordinate Space=::Local
   )
   &SolveForcesAndVelocity(Clamp Velocity=false, Limit Acceleration=false)
   &GenerateDeathEvent(Use Event Probability=false)
   &SubUVAnimation(
    Random Start Frame=true
    "Sprite Renderer "=DataInterface() {propertyValues {SpriteRenderer=&self:Bubbles_0.NiagaraSpriteRendererProperties_0}}
    UseEndFrame=false
    UseStartFrame=false
   )
   &ScaleColor(
    Scale RGB=&VectorFromFloat(Value=&MergeHelper_FloatSwitch(
      0=&RandomRangeFloat(Minimum=0.4, Maximum=0.6)
      1=1
      2=1
      3=?
      Index=&SpawnGroup()
      Num=::NewEnumerator1
     ))
    ScaleA=true
    ScaleRGB=true
   )
  ]
  rendererClasses [&NiagaraSpriteRendererProperties]
  renderers [
   {rendererClass=&NiagaraSpriteRendererProperties}
   {rendererClass=&NiagaraSpriteRendererProperties, rendererIndex=1}
   {rendererClass=&NiagaraSpriteRendererProperties, rendererIndex=2}
  ]
 }
 Explosion {
  emitterUpdateScript [
   &EmitterState(Life Cycle Mode=::NewEnumerator1, Loop Duration=2, UseLoopDelay=false)
   &VC_FilteredSocketCount()
   &SpawnRate(SpawnRate=40, Use Spawn Probability=false) {enabled=false}
  ]
  particleSpawnScript [
   &VC_PaintedVariantsOverride() {enabled=false}
   &InitializeParticle(
    Lifetime Max=0.5
    Lifetime Min=0.3
    Lifetime Mode=::NewEnumerator1
    Sprite Rotation Angle=&RandomRangeFloat(Minimum=0, Maximum=1)
    Sprite Rotation Mode=::NewEnumerator1
    Sprite Size Mode=::NewEnumerator3
    Uniform Sprite Size=60
    UsePositionOffset=false
    Write Material Random=true
   )
   &PartitionParticles(Number Of Partitions=$Emitter.PortCount)
   &SkeletalMeshLocation(
    Bone / Socket Index=$Particles.PartitionParticles.Partition
    Bone / Socket Sampling Mode=::NewEnumerator5
    MaskSpawnGroup=false
    Mesh Sampling Type=::NewEnumerator6
    Skeletal Mesh=$System.SkeletalMesh
   )
   &AddVelocity(Velocity=<0 0 10>)
   &AccelerationForce(Acceleration=&RandomRangeVector(Minimum=<0 0 -150>, Maximum=<0 0 -200>)) {enabled=false}
   &ApplyInitialForces()
  ]
  particleUpdateScript [
   &ParticleState()
   &ScaleSpriteSize(Uniform Curve Sprite Scale=DataInterface() {propertyValues {Curve {keys [{tangentMode=RCTM_User, time=0, value=0.2}, {tangentMode=RCTM_User, time=1, value=4}]}}})
   &ScaleColor(
    Scale Alpha=&FloatFromCurve(FloatCurve=DataInterface() {propertyValues {Curve {keys [<at -0.0046 0.5 RCIM_Cubic>,<at 1.0011 0.002 RCIM_Cubic>]}}})
    Scale RGB=<2 2 2>
    ScaleA=true
    ScaleRGB=true
   )
   &SolveForcesAndVelocity(Clamp Velocity=false, Limit Acceleration=false)
   &SubUVAnimation(
    "Sprite Renderer "=DataInterface() {propertyValues {SpriteRenderer=&self:Explosion_0.NiagaraSpriteRendererProperties_0}}
    UseEndFrame=false
    UseStartFrame=false
   )
  ]
  rendererClasses [&NiagaraSpriteRendererProperties]
  renderers [{rendererClass=&NiagaraSpriteRendererProperties}]
 }
 Jellyfish {
  emitterUpdateScript [
   &EmitterState()
   &VC_FilteredSocketCount()
   &SpawnRate(
    SpawnRate=&MakeCustomFloatFromBool(Bool=$User.bIsFrontendPreview, True Float=40, False Float=100)
    Use Spawn Probability=false
   )
   &SpawnRate(SpawnGroup=1, SpawnRate=40, Use Spawn Probability=false) {module=SpawnRate001}
  ]
  particleSpawnScript [
   /SetVariables(Particles.VisibilityTag=$Emitter.SpawnGroup) {isSetParameters=true}
   &VC_PaintedVariantsOverride() {enabled=false}
   &InitializeParticle(
    Lifetime=&MakeCustomFloatFromBool(Bool=$User.bIsFrontendPreview, True Float=&RandomRangeFloat(Minimum=0.3, Maximum=0.7), False Float=0.3)
    Mesh Scale Mode=::NewEnumerator3
    Mesh Uniform Scale=&RandomRangeFloat(Minimum=0.1, Maximum=0.2)
    UsePositionOffset=false
    Write Material Random=true
   )
   &PartitionParticles(Number Of Partitions=$Emitter.PortCount)
   &SkeletalMeshLocation(
    Bone / Socket Index=$Particles.PartitionParticles.Partition
    Bone / Socket Sampling Mode=::NewEnumerator5
    MaskSpawnGroup=false
    Mesh Sampling Type=::NewEnumerator6
    Sampled Position Offset=<-20 0 0>
    Skeletal Mesh=$System.SkeletalMesh
   )
   &AddVelocity(Velocity=&MakeCustomVectorFromBool(Bool=$User.bIsFrontendPreview, True Vector=<-500 0 0>, False Vector=<0 0 0>))
   &ShapeLocation(MaskSpawnGroup=false, Override Seed=false, Sphere Radius=35, UseOwnerScale=true)
   &InitialMeshOrientation(
    Mesh Axis To Orient=<0 0 0>
    Mesh Orientation Mode=::NewEnumerator2
    Orientation Coordinate Space=::Local
    Orientation Vector=<0 0 -1>
    Rotation=&RandomRangeVector(Minimum=<0 0 0>, Maximum=<1 0 0>)
    Rotation Coordinate Space=::NewEnumerator2
    Use Rotation Vector=false
   )
  ]
  particleUpdateScript [
   &ParticleState()
   &ScaleMeshSize(Scale Factor=&VectorFromCurve(
     VectorCurve=DataInterface() {propertyValues {
       XCurve {
        defaultValue=1
        keys [
         <at 0 0 RCIM_Cubic>
         <at 0.1 0.3>
         {arriveTangentWeight=0.2, interpMode=RCIM_Cubic, tangentMode=RCTM_User, time=0.4, value=1}
         {arriveTangentWeight=0.0667, interpMode=RCIM_Cubic, tangentMode=RCTM_User, time=0.9, value=0.4}
         <at 1 0 RCIM_Cubic>
        ]
       }
       YCurve {
        defaultValue=1
        keys [
         <at 0 0 RCIM_Cubic>
         <at 0.1 0.3>
         {arriveTangentWeight=0.2, interpMode=RCIM_Cubic, tangentMode=RCTM_User, time=0.4, value=1}
         {arriveTangentWeight=0.0667, interpMode=RCIM_Cubic, tangentMode=RCTM_User, time=0.9, value=0.4}
         <at 1 0 RCIM_Cubic>
        ]
       }
       ZCurve {
        defaultValue=1
        keys [
         <at 0 0 RCIM_Cubic>
         {arriveTangentWeight=0.0333, interpMode=RCIM_Cubic, tangentMode=RCTM_User, time=0.1, value=1.3}
         {arriveTangentWeight=0.1667, interpMode=RCIM_Cubic, tangentMode=RCTM_User, time=0.3, value=0.5}
         {arriveTangentWeight=0.0333, interpMode=RCIM_Cubic, tangentMode=RCTM_User, time=0.4, value=0.5}
         {arriveTangentWeight=0.1452, interpMode=RCIM_Cubic, leaveTangentWeight=0.1692, tangentMode=RCTM_User, time=0.6309, value=1.082}
         {arriveTangentWeight=0.0667, interpMode=RCIM_Cubic, tangentMode=RCTM_User, time=0.9, value=0.4}
         <at 1 0 RCIM_Cubic>
        ]
       }
      }}
     CurveIndex=&MakeCustomFloatFromBool(Bool=$User.bIsFrontendPreview, True Float=$Particles.NormalizedAge, False Float=$Particles.Lifetime)
    )) {module=ScaleMeshSize001}
   &CurlNoiseForce(Bias Noise=false, Noise Frequency=30, Noise Strength=500, Pan Noise=false)
   &Drag(
    Drag=&MakeCustomFloatFromBool(Bool=true, True Float=1, False Float=10)
    Use Linear Drag=true
    Use Rotational Drag=true
   )
   &AccelerationForce(
    Acceleration=&MakeCustomVectorFromBool(Bool=$User.bIsFrontendPreview, True Vector=<-2000 0 0>, False Vector=<0 0 0>)
    Coordinate Space=::Local
   )
   &SolveForcesAndVelocity(Clamp Velocity=false, Limit Acceleration=false)
  ]
  rendererClasses [&NiagaraMeshRendererProperties, &NiagaraSpriteRendererProperties]
  renderers [
   {rendererClass=&NiagaraMeshRendererProperties, rendererIndex=0}
   {rendererClass=&NiagaraMeshRendererProperties, rendererIndex=1}
   {rendererClass=&NiagaraSpriteRendererProperties, rendererIndex=2}
  ]
 }
}
summary {
 systemName=NS_VFX_LionsMane01
 emitters [
  {emitterName=Bubbles, rendererClasses [&NiagaraSpriteRendererProperties]}
  {emitterName=Explosion, rendererClasses [&NiagaraSpriteRendererProperties]}
  {emitterName=Jellyfish, rendererClasses [&NiagaraMeshRendererProperties, &NiagaraSpriteRendererProperties]}
 ]
 userVariables [
  User.bIsFrontendPreview {type=NiagaraBool}
  User.bIsPainted {type=NiagaraBool}
  User.Brightness {type=NiagaraFloat, defaultValue=1, description=""}
  User.ColorVariantID {type=NiagaraInt32}
  User.DTCarOffset {type=Vector3f, description="", defaultValue {struct=Vector3f, value=<0 0 0>}}
  User.PaintedColor {type=LinearColor, description="", defaultValue {struct=LinearColor, value=<rgba 1 1 1 1>}}
  User.PaintedColor2 {type=LinearColor, description="", defaultValue {struct=LinearColor, value=<rgba 1 1 1 1>}}
  User.PaintedScalar {type=NiagaraFloat, defaultValue=1, description=""}
  User.PortCount {type=NiagaraInt32, defaultValue=1}
 ]
}
